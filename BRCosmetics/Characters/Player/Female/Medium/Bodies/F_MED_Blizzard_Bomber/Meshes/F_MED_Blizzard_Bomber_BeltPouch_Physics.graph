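# Physics asset for the belt pouch on the female medium Blizzard Bomber player cosmetic body.
# A box body covers the dyn_pouch_01 bone, alongside capsule bodies on the pelvis and right thigh.
# A single constraint joins the pouch to the right thigh with tight swing and twist limits, restricting how far it can rotate relative to the leg.
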PhysicsAsset /BRCosmetics/Characters/Player/Female/Medium/Bodies/F_MED_Blizzard_Bomber/Meshes/F_MED_Blizzard_Bomber_BeltPouch_Physics
bodies {
 dyn_pouch_01 [{
   shapeType=Box
   extentX=4.5
   extentY=10.5
   extentZ=15.5
   shapeName=dyn_pouch_01_box
   center=<0.2896 -0.7196 2.8503>
   rotation=<rot -7.0283 -0.0157 -0.4368>
  }]
 pelvis [{
   shapeType=Capsule
   length=5.9306
   radius=11.338
   shapeName=pelvis_capsule
   center=<-0.8964 -1.298 1.5663>
   rotation=<rot -5.401 -90.0002 -5.3511>
  }]
 thigh_r [{
   shapeType=Capsule
   length=32.75
   radius=8
   shapeName=thigh_r_capsule
   center=<15.3899 -0.0468 -0.6887>
   rotation=<rot -89.7215 162.779 -161.7841>
  }]
}
constraints [{bone1Name=dyn_pouch_01, bone2Name=thigh_r, swing1LimitDegrees=10, swing2LimitDegrees=20, twistLimitDegrees=5}]
