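# Physics asset for the Blizzard Bomber jacket on the medium female player character, covering the upper body and the dynamic collar bones.
# Capsules cover spine, neck, clavicles and arms; neck, clavicles and upper arms constrain to spine_05, while the forearms have no constraints.
# Each collar side chains MSTR, 01 and 02 bones from spine_05, with looser limits at the MSTR root than along the chain.
# Left and right capsules differ in size, most visibly at the clavicles.
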PhysicsAsset /BRCosmetics/Characters/Player/Female/Medium/Bodies/F_MED_Blizzard_Bomber/Meshes/F_MED_Blizzard_Bomber_Jacket_Physics
bodies {
 clavicle_l [{shapeType=Capsule, length=7.25, radius=7.5, shapeName=clavicle_l_capsule, center=<7.4467 0 -1.3131>, rotation=<rot -80 179.9991 -179.9992>}]
 clavicle_r [{shapeType=Capsule, length=3, radius=3.5, shapeName=clavicle_r_capsule, center=<-9.3126 0 0.8951>, rotation=<rot -80 179.9991 -179.9993>}]
 dyn_collar_l_01 [{shapeType=Capsule, length=0, radius=3, shapeName=dyn_collar_l_01_capsule, center=<0.1931 0.4128 -0.4919>, rotation=<rot 0 -0.0001 49.9999>}]
 dyn_collar_l_02 [{
   shapeType=Capsule
   length=5.25
   radius=5
   shapeName=dyn_collar_l_02_capsule
   center=<-0.2834 -1.5878 2.0527>
   rotation=<rot -37.159 -115.4146 26.0132>
  }]
 dyn_collar_l_MSTR [{shapeType=Capsule, length=0.505, radius=0.505, shapeName=dyn_collar_l_MSTR_capsule, rotation=<rot 0 -90 0>}]
 dyn_collar_r_01 [{shapeType=Capsule, length=0, radius=3, shapeName=dyn_collar_r_01_capsule, center=<0.1931 0.3211 -0.5561>, rotation=<rot 0.0001 0 60.0001>}]
 dyn_collar_r_02 [{
   shapeType=Capsule
   length=5.5
   radius=4.25
   shapeName=dyn_collar_r_02_capsule
   center=<-0.9954 1.584 2.4513>
   rotation=<rot -37.1592 115.4144 -26.013>
  }]
 dyn_collar_r_MSTR [{shapeType=Capsule, length=0.505, radius=0.505, shapeName=dyn_collar_r_MSTR_capsule, rotation=<rot 0 -90 0>}]
 lowerarm_l [{shapeType=Capsule, length=19, radius=7.75, shapeName=lowerarm_l_capsule, center=<5.8159 0 0>, rotation=<rot -90 180 180>}]
 lowerarm_r [{shapeType=Capsule, length=19, radius=7.75, shapeName=lowerarm_r_capsule, center=<-5.4558 0 0>, rotation=<rot -90 179.9604 -179.9604>}]
 neck_01 [{shapeType=Capsule, length=3.25, radius=5, shapeName=neck_01_capsule, rotation=<rot -90 13.6118 -13.612>}]
 spine_05 [{shapeType=Capsule, length=6, radius=10.5, shapeName=spine_05_capsule, center=<-0.381 0 0>}]
 upperarm_l [{shapeType=Capsule, length=14.25, radius=8.5, shapeName=upperarm_l_capsule, center=<10.2467 -1.1497 0>, rotation=<rot -90 0 -10>}]
 upperarm_r [{shapeType=Capsule, length=15, radius=8.5, shapeName=upperarm_r_capsule, center=<-9.2226 0.507 0>, rotation=<rot -90 0 -10>}]
}
constraints [
 {bone1Name=neck_01, bone2Name=spine_05, swing1LimitDegrees=20, swing2LimitDegrees=20, twistLimitDegrees=5}
 {bone1Name=clavicle_l, bone2Name=spine_05, swing1LimitDegrees=20, swing2LimitDegrees=20, twistLimitDegrees=5}
 {bone1Name=upperarm_l, bone2Name=spine_05, swing1LimitDegrees=20, swing2LimitDegrees=20, twistLimitDegrees=5}
 {bone1Name=clavicle_r, bone2Name=spine_05, swing1LimitDegrees=20, swing2LimitDegrees=20, twistLimitDegrees=5}
 {bone1Name=upperarm_r, bone2Name=spine_05, swing1LimitDegrees=20, swing2LimitDegrees=20, twistLimitDegrees=5}
 {bone1Name=dyn_collar_l_01, bone2Name=dyn_collar_l_MSTR, swing1LimitDegrees=20, swing2LimitDegrees=20, twistLimitDegrees=5}
 {bone1Name=dyn_collar_l_02, bone2Name=dyn_collar_l_01, swing1LimitDegrees=20, swing2LimitDegrees=20, twistLimitDegrees=5}
 {bone1Name=dyn_collar_r_01, bone2Name=dyn_collar_r_MSTR, swing1LimitDegrees=20, swing2LimitDegrees=20, twistLimitDegrees=5}
 {bone1Name=dyn_collar_r_02, bone2Name=dyn_collar_r_01, swing1LimitDegrees=20, swing2LimitDegrees=20, twistLimitDegrees=5}
 {bone1Name=dyn_collar_r_MSTR, bone2Name=spine_05, swing1LimitDegrees=45, swing2LimitDegrees=45, twistLimitDegrees=45}
 {bone1Name=dyn_collar_l_MSTR, bone2Name=spine_05, swing1LimitDegrees=45, swing2LimitDegrees=45, twistLimitDegrees=45}
]
