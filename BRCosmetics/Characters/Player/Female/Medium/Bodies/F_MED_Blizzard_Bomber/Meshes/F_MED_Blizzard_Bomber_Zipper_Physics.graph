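# Physics asset for the dangling zipper on the female medium Blizzard Bomber body cosmetic.
# Capsule bodies on spine_05, neck_01, dyn_zipperMSTR and dyn_zipper form a constraint chain from upper spine to zipper pull.
# Rotation is locked from spine through neck to the zipper master, leaving the final zipper joint as the only one able to swing and twist.
# The large spine and neck capsules approximate the upper torso around the tiny zipper capsules.
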PhysicsAsset /BRCosmetics/Characters/Player/Female/Medium/Bodies/F_MED_Blizzard_Bomber/Meshes/F_MED_Blizzard_Bomber_Zipper_Physics
bodies {
 dyn_zipper [{
   shapeType=Capsule
   length=1.2621
   radius=1.1033
   shapeName=dyn_zipper_capsule
   center=<0.7124 -0.417 0.0109>
   rotation=<rot 0 90.5701 -89.2395>
  }]
 dyn_zipperMSTR [{shapeType=Capsule, length=0.505, radius=0.505, shapeName=dyn_zipperMSTR_capsule, center=<0 -0.3264 0>, rotation=<rot 0 -90 0>}]
 neck_01 [{shapeType=Capsule, length=5, radius=4.7706, shapeName=neck_01_capsule, rotation=<rot 90 0 0>}]
 spine_05 [{shapeType=Capsule, length=16.6, radius=14.4715, shapeName=spine_05_capsule, center=<-5.3627 0 0>}]
}
constraints [
 {
  bone1Name=neck_01
  bone2Name=spine_05
  swing1LimitDegrees=45
  swing1Motion=Locked
  swing2LimitDegrees=45
  swing2Motion=Locked
  twistLimitDegrees=45
  twistMotion=Locked
 }
 {
  bone1Name=dyn_zipperMSTR
  bone2Name=neck_01
  swing1LimitDegrees=45
  swing1Motion=Locked
  swing2LimitDegrees=45
  swing2Motion=Locked
  twistLimitDegrees=45
  twistMotion=Locked
 }
 {bone1Name=dyn_zipper, bone2Name=dyn_zipperMSTR, swing1LimitDegrees=50, swing2LimitDegrees=30, twistLimitDegrees=5}
]
