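# Post-process style Anim Blueprint for the female medium Blizzard Bomber cosmetic body, copying the attached parent's pose and animating jacket, collar, zipper, belt pouch and kneepads.
# The copied pose feeds a deform Control Rig, weighted bone constraints, three RigidBody simulations, then corrective bone offsets.
# Neck scale and zipper offset apply only in the front end.
# EventGraph derives head, arm and thigh values unused by AnimGraph, which also holds disconnected constraint and offset nodes.
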AnimBlueprint /BRCosmetics/Characters/Player/Female/Medium/Bodies/F_MED_Blizzard_Bomber/Meshes/F_MED_Blizzard_Bomber_AnimBP
target_skeleton=&/F_MED_Blizzard_Bomber_Skeleton
graphs [
 AnimGraph {nodes [
   "Constraint - dyn_pouch_01 (PPPPPP)"!(
    ComponentPose
    Alpha=1
    ConstraintWeights_0=0.05
    ConstraintWeights_1:float=0.05
    ConstraintWeights_2:float=0.5
    ConstraintWeights_3:float=0.01
    ConstraintWeights_4:float=0.15
    ConstraintWeights_5:float=0.24
   ) -> (Pose) {
    alphaBoolBlend {blendOption=Linear}
    boneToModify {boneName=dyn_pouch_01}
    constraintSetup [
     {targetBone {boneName=dyn_belt_fr_03}}
     {targetBone {boneName=dyn_belt_fr_04}}
     {targetBone {boneName=dyn_pouch_01}}
     {targetBone {boneName=dyn_belt_bck_03}}
     {targetBone {boneName=dyn_belt_bck_04}}
     {targetBone {boneName=thigh_l}}
    ]
    constraintWeights [1,1,1,1,1,1]
   }
   "Constraint - dyn_pouch_01 (PP)"!(ComponentPose, Alpha=1, ConstraintWeights_0=0.75, ConstraintWeights_1:float=0.25) -> (Pose) {
    alphaBoolBlend {blendInTime=0.1, blendOption=QuarticInOut, blendOutTime=0.1}
    alphaScaleBiasClamp {bInterpResult=true, interpSpeedDecreasing=5, inRange {max=0, min=1}}
    boneToModify {boneName=dyn_pouch_01}
    constraintSetup [{perAxis {bY=false, bZ=false}, targetBone {boneName=thigh_r}}, {perAxis {bY=false, bZ=false}, targetBone {boneName=pelvis}}]
    constraintWeights [1,1]
   }
   "Constraint - dyn_zipper (PPPPP)"!(
    ComponentPose
    Alpha=1
    ConstraintWeights_0=0.9999
    ConstraintWeights_1:float=0
    ConstraintWeights_2:float=0
    ConstraintWeights_3:float=0
    ConstraintWeights_4:float=0
   ) -> (Pose) {
    alphaBoolBlend {blendOption=Linear}
    boneToModify {boneName=dyn_zipper}
    constraintSetup [
     {targetBone {boneName=spine_05}}
     {targetBone {boneName=clavicle_l}}
     {targetBone {boneName=clavicle_r}}
     {targetBone {boneName=dyn_zipper}}
     {targetBone {boneName=dyn_collar_l_01}}
    ]
    constraintWeights [1,1,1,1,1]
   }
   "Transform (Modify) Bone - Bone: dyn_belt_fr_02"(Alpha=1, Translation="0.542,1.405,1.825", Rotation="-0.000,0.000,10.000", Scale="1,1,1") {
    rotationMode=BMM_Additive
    translationMode=BMM_Additive
    alphaBoolBlend {blendOption=Linear}
    boneToModify {boneName=dyn_belt_fr_02}
    rotation {roll=10}
    translation=<0.542 1.405 1.825>
   }
   n1 = "Transform (Modify) Bone - Bone: dyn_pouch_02"(Alpha=1, Translation="-1,0,0", Rotation="10,0,0", Scale="1,1,1") {
    rotationMode=BMM_Additive
    rotationSpace=BCS_ParentBoneSpace
    translationMode=BMM_Additive
    alphaBoolBlend {blendOption=Linear}
    alphaScaleBiasClamp {bInterpResult=true, bMapRange=true, clampMin=1, interpSpeedDecreasing=2, inRange {max=0.8, min=0.2}}
    boneToModify {boneName=dyn_pouch_02}
    rotation {pitch=10}
    translation {x=-1}
   }
   "Transform (Modify) Bone - Bone: dyn_pouch_02"(ComponentPose=@n1, Alpha=1, Translation="-1,0,0", Rotation="10,0,0", Scale="1,1,1") {
    rotationMode=BMM_Additive
    rotationSpace=BCS_ParentBoneSpace
    translationMode=BMM_Additive
    alphaBoolBlend {blendOption=Linear}
    alphaScaleBiasClamp {
     bClampResult=true
     bInterpResult=true
     bMapRange=true
     clampMin=1
     interpSpeedDecreasing=2
     inRange {max=-0.2, min=-0.8}
     outRange {max=0, min=1}
    }
    boneToModify {boneName=dyn_pouch_02}
    rotation {pitch=10}
    translation {x=-1}
   }
   n2 = "Transform (Modify) Bone - Bone: dyn_belt_bck_02"(Alpha=1, Translation="0,0,1", Rotation="0,0,0", Scale="1,1,1") {translationMode=BMM_Additive, alphaBoolBlend {blendOption=Linear}, boneToModify {boneName=dyn_belt_bck_02}, translation {z=1}}
   "Transform (Modify) Bone - Bone: dyn_belt_bck_03"(ComponentPose=@n2, Alpha=1, Translation="0.000,0.000,0.109", Rotation="0,0,0", Scale="1,1,1") {translationMode=BMM_Additive, alphaBoolBlend {blendOption=Linear}, boneToModify {boneName=dyn_belt_bck_03}, translation {z=0.109}}
   "Transform (Modify) Bone - Bone: dyn_pouch_02"(Alpha=1, Translation="-0.5,0,0", Rotation="0,5,0", Scale="1,1,1") {
    translationMode=BMM_Additive
    alphaBoolBlend {blendOption=Linear}
    alphaScaleBiasClamp {clampMin=1}
    boneToModify {boneName=dyn_pouch_02}
    rotation {yaw=5}
    translation {x=-0.5}
   }
   n3 = "Transform (Modify) Bone - Bone: dyn_belt_fr_02"(Alpha=1, Translation="-2,0.000,3.000", Rotation="0,0,0", Scale="1,1,1") {
    rotationMode=BMM_Additive
    translationMode=BMM_Additive
    alphaBoolBlend {blendOption=Linear}
    boneToModify {boneName=dyn_belt_fr_02}
    translation {x=-2, z=3}
   }
   "Transform (Modify) Bone - Bone: dyn_belt_fr_03"(ComponentPose=@n3, Alpha=1, Translation="-2,-0.000,3.000", Rotation="0,0,0", Scale="1,1,1") {
    rotationMode=BMM_Additive
    translationMode=BMM_Additive
    alphaBoolBlend {blendOption=Linear}
    boneToModify {boneName=dyn_belt_fr_03}
    translation {x=-2, z=3}
   }
   n4 = "Transform (Modify) Bone - Bone: dyn_belt_fr_02"(Alpha=1, Translation="0.000,0.000,1.312", Rotation="0,0,0", Scale="1,1,1") {translationMode=BMM_Additive, alphaBoolBlend {blendOption=Linear}, boneToModify {boneName=dyn_belt_fr_02}, translation {z=1.312}}
   n5 = "Transform (Modify) Bone - Bone: dyn_belt_fr_03"(ComponentPose=@n4, Alpha=1, Translation="0.000,1.435,1.831", Rotation="0,0,0", Scale="1,1,1") {
    rotationMode=BMM_Additive
    translationMode=BMM_Additive
    alphaBoolBlend {blendOption=Linear}
    boneToModify {boneName=dyn_belt_fr_03}
    translation {y=1.435, z=1.831}
   }
   "Transform (Modify) Bone - Bone: dyn_belt_fr_04"(ComponentPose=@n5, Alpha=1, Translation="-3.350,0.356,1.324", Rotation="0,0,0", Scale="1,1,1") {translationMode=BMM_Additive, alphaBoolBlend {blendOption=Linear}, boneToModify {boneName=dyn_belt_fr_04}, translation=<-3.35 0.356 1.324>}
   n6 = "Transform (Modify) Bone - Bone: dyn_belt_bck_02"(Alpha=1, Translation="0,0,0", Rotation="0,0,0", Scale="1,1,1") {rotationMode=BMM_Additive, translationMode=BMM_Additive, alphaBoolBlend {blendOption=Linear}, boneToModify {boneName=dyn_belt_bck_02}}
   n7 = "Transform (Modify) Bone - Bone: dyn_belt_bck_03"(ComponentPose=@n6, Alpha=1, Translation="0,0,0", Rotation="0,0,0", Scale="1,1,1") {rotationMode=BMM_Additive, translationMode=BMM_Additive, alphaBoolBlend {blendOption=Linear}, boneToModify {boneName=dyn_belt_bck_03}}
   "Transform (Modify) Bone - Bone: dyn_belt_bck_04"(ComponentPose=@n7, Alpha=1, Translation="-1,0,0", Rotation="0,0,0", Scale="1,1,1") {
    rotationMode=BMM_Additive
    translationMode=BMM_Additive
    alphaBoolBlend {blendOption=Linear}
    boneToModify {boneName=dyn_belt_bck_04}
    translation {x=-1}
   }
   n8 = "Transform (Modify) Bone - Bone: dyn_belt_bck_02"(Alpha=1, Translation="0,0,0", Rotation="0,0,0", Scale="1,1,1") {rotationMode=BMM_Additive, translationMode=BMM_Additive, alphaBoolBlend {blendOption=Linear}, boneToModify {boneName=dyn_belt_bck_02}}
   n9 = "Transform (Modify) Bone - Bone: dyn_belt_bck_03"(ComponentPose=@n8, Alpha=1, Translation="0,0,0", Rotation="0,0,0", Scale="1,1,1") {rotationMode=BMM_Additive, translationMode=BMM_Additive, alphaBoolBlend {blendOption=Linear}, boneToModify {boneName=dyn_belt_bck_03}}
   n10 = "Transform (Modify) Bone - Bone: dyn_belt_bck_04"(ComponentPose=@n9, Alpha=1, Translation="-1,0,0", Rotation="0,0,0", Scale="1,1,1") {
    rotationMode=BMM_Additive
    translationMode=BMM_Additive
    alphaBoolBlend {blendOption=Linear}
    boneToModify {boneName=dyn_belt_bck_04}
    translation {x=-1}
   }
   n11 = "Transform (Modify) Bone - Bone: dyn_belt_fr_02"(ComponentPose=@n10, Alpha=1, Translation="-2,0,3", Rotation="0,0,0", Scale="1,1,1") {
    rotationMode=BMM_Additive
    translationMode=BMM_Additive
    alphaBoolBlend {blendOption=Linear}
    boneToModify {boneName=dyn_belt_fr_02}
    translation {x=-2, z=3}
   }
   "Transform (Modify) Bone - Bone: dyn_belt_fr_03"(ComponentPose=@n11, Alpha=1, Translation="-2,0,3", Rotation="0,0,0", Scale="1,1,1") {
    rotationMode=BMM_Additive
    translationMode=BMM_Additive
    alphaBoolBlend {blendOption=Linear}
    boneToModify {boneName=dyn_belt_fr_03}
    translation {x=-2, z=3}
   }
   n12 = Control Rig(Source=Copy Pose From Mesh() {bUseAttachedParent=true}) {
    alphaBoolBlend {blendOption=Linear}
    controlRigAssetReference {blueprintRigClass=&/F_MED_Blizzard_Bomber_DeformRig_ControlRig.F_MED_Blizzard_Bomber_DeformRig_ControlRig_C}
   }
   n13 = "Constraint - dyn_collar_l_02 (PPPPP)"!(
    ComponentPose=Local To Component(LocalPose=@n12)
    Alpha=1
    ConstraintWeights_0=0.0003
    ConstraintWeights_1:float=0.1038
    ConstraintWeights_2:float=0.0032
    ConstraintWeights_3:float=0.6061
    ConstraintWeights_4:float=0.2866
   ) -> (Pose) {
    alphaBoolBlend {blendOption=Linear}
    boneToModify {boneName=dyn_collar_l_02}
    constraintSetup [
     {targetBone {boneName=spine_04}}
     {targetBone {boneName=spine_05}}
     {targetBone {boneName=clavicle_l}}
     {targetBone {boneName=dyn_collar_l_01}}
     {targetBone {boneName=dyn_collar_l_02}}
    ]
    constraintWeights [1,1,1,1,1]
   }
   n14 = "Constraint - dyn_collar_l_02 (PPPPPP)"!(
    ComponentPose=@n13
    Alpha=1
    ConstraintWeights_0=0.0056
    ConstraintWeights_1:float=0.2174
    ConstraintWeights_2:float=0.0451
    ConstraintWeights_3:float=0.0005
    ConstraintWeights_4:float=0.1819
    ConstraintWeights_5:float=0.5494
   ) -> (Pose) {
    alphaBoolBlend {blendOption=Linear}
    boneToModify {boneName=dyn_collar_l_02}
    constraintSetup [
     {targetBone {boneName=spine_04}}
     {targetBone {boneName=spine_05}}
     {targetBone {boneName=clavicle_r}}
     {targetBone {boneName=upperarm_twist_01_r}}
     {targetBone {boneName=dyn_collar_r_02}}
     {targetBone {boneName=dyn_collar_r_01}}
    ]
    constraintWeights [1,1,1,1,1,1]
   }
   n15 = "Constraint - dyn_kneepad_l (PPP)"!(ComponentPose=@n14, Alpha=1, ConstraintWeights_0=0.4338, ConstraintWeights_1:float=0.5662, ConstraintWeights_2:float=0) -> (Pose) {
    alphaBoolBlend {blendOption=Linear}
    boneToModify {boneName=dyn_kneepad_l}
    constraintSetup [{targetBone {boneName=thigh_l}}, {targetBone {boneName=calf_l}}, {targetBone {boneName=calf_twist_02_l}}]
    constraintWeights [1,1,1]
   }
   n16 = "Constraint - dyn_kneepad_r (PPP)"!(ComponentPose=@n15, Alpha=1, ConstraintWeights_0=0.4338, ConstraintWeights_1:float=0.5662, ConstraintWeights_2:float=0) -> (Pose) {
    alphaBoolBlend {blendOption=Linear}
    boneToModify {boneName=dyn_kneepad_r}
    constraintSetup [{targetBone {boneName=thigh_r}}, {targetBone {boneName=calf_r}}, {targetBone {boneName=calf_twist_02_r}}]
    constraintWeights [1,1,1]
   }
   n17 = "Constraint - dyn_zipperMSTR (PPPPPPP)"!(
    ComponentPose=@n16
    Alpha=1
    ConstraintWeights_0=0.818
    ConstraintWeights_1:float=0.0004
    ConstraintWeights_2:float=0.0007
    ConstraintWeights_3:float=0.0001
    ConstraintWeights_4:float=0.039
    ConstraintWeights_5:float=0.0788
    ConstraintWeights_6:float=0.0629
   ) -> (Pose) {
    alphaBoolBlend {blendOption=Linear}
    boneToModify {boneName=dyn_zipperMSTR}
    constraintSetup [
     {targetBone {boneName=spine_05}}
     {targetBone {boneName=clavicle_l}}
     {targetBone {boneName=clavicle_r}}
     {targetBone {boneName=dyn_collar_l_01}}
     {targetBone {boneName=neck_02}}
     {targetBone {boneName=neck_01}}
     {targetBone {boneName=head}}
    ]
    constraintWeights [1,1,1,1,1,1,1]
   }
   n18 = RigidBody(ComponentPose=@n17) {
    bClampLinearTranslationLimitToRefPose=true
    bOverrideWorldGravity=true
    overridePhysicsAsset=&/F_MED_Blizzard_Bomber_BeltPouch_Physics
    simulationSpace=BaseBoneSpace
    alphaBoolBlend {blendOption=Linear}
    baseBoneRef {boneName=pelvis}
    componentAppliedLinearAccClamp=<3000 3000 3000>
    componentLinearAccScale=<0.7 0.7 0.7>
    componentLinearVelScale=<0.1 0.1 0.1>
    overrideWorldGravity {z=-980}
    simSpaceSettings {maxAngularAcceleration=10, maxAngularVelocity=30, maxLinearAcceleration=10, maxLinearVelocity=3, externalLinearDragV=<1 1 1>}
   }
   n19 = RigidBody(ComponentPose=@n18) {
    bClampLinearTranslationLimitToRefPose=true
    bOverrideWorldGravity=true
    overridePhysicsAsset=&/F_MED_Blizzard_Bomber_Jacket_Physics
    simulationSpace=BaseBoneSpace
    alphaBoolBlend {blendOption=Linear}
    baseBoneRef {boneName=spine_05}
    componentAppliedLinearAccClamp=<3000 3000 3000>
    componentLinearAccScale=<0.2 0.2 0.2>
    componentLinearVelScale=<0.1 0.1 0.1>
    overrideWorldGravity {z=-980}
    simSpaceSettings {maxAngularAcceleration=10, maxAngularVelocity=100, maxLinearAcceleration=10, maxLinearVelocity=50, externalLinearDragV=<1 1 1>}
   }
   n20 = RigidBody(ComponentPose=@n19) {
    bClampLinearTranslationLimitToRefPose=true
    bOverrideWorldGravity=true
    overridePhysicsAsset=&/F_MED_Blizzard_Bomber_Zipper_Physics
    simulationSpace=BaseBoneSpace
    alphaBoolBlend {blendOption=Linear}
    baseBoneRef {boneName=neck_01}
    componentAppliedLinearAccClamp=<3000 3000 3000>
    componentLinearAccScale=<0.7 0.7 0.7>
    componentLinearVelScale=<0.2 0.2 0.2>
    overrideWorldGravity {z=-980}
   }
   n21 = "Constraint - dyn_pouch_01 (PP)"(ComponentPose=@n20, Alpha=1, ConstraintWeights_0=0.25, ConstraintWeights_1=0.75) {
    alphaBoolBlend {blendOption=Linear}
    boneToModify {boneName=dyn_pouch_01}
    constraintSetup [{targetBone {boneName=pelvis}}, {targetBone {boneName=dyn_pouch_01}}]
    constraintWeights [1,1]
   }
   n22 = "Constraint - dyn_pouch_01 (PPPPP)"!(
    ComponentPose=@n21
    Alpha=1
    ConstraintWeights_0=0.0958
    ConstraintWeights_1:float=0.1776
    ConstraintWeights_2:float=0.3955
    ConstraintWeights_3:float=0.033
    ConstraintWeights_4:float=0.298
   ) -> (Pose) {
    alphaBoolBlend {blendOption=Linear}
    alphaScaleBiasClamp {bInterpResult=true}
    boneToModify {boneName=dyn_pouch_01}
    constraintSetup [
     {targetBone {boneName=dyn_belt_fr_03}}
     {targetBone {boneName=dyn_belt_fr_04}}
     {targetBone {boneName=dyn_pouch_01}}
     {targetBone {boneName=dyn_belt_bck_03}}
     {targetBone {boneName=dyn_belt_bck_04}}
    ]
    constraintWeights [1,1,1,1,1]
   }
   n23 = "Transform (Modify) Bone - Bone: dyn_pouch_02"(ComponentPose=@n22, Alpha=1, Translation="-0.5,-1,0", Rotation="5,0,0", Scale="1,1,1") {
    rotationMode=BMM_Additive
    rotationSpace=BCS_ParentBoneSpace
    translationMode=BMM_Additive
    alphaBoolBlend {blendOption=Linear}
    alphaScaleBiasClamp {clampMin=1}
    boneToModify {boneName=dyn_pouch_02}
    rotation {pitch=5}
    translation=<-0.5 -1>
   }
   n24 = "Transform (Modify) Bone - Bone: dyn_pouch_02"(ComponentPose=@n23, Alpha=1, Translation="-2,0,3", Rotation="-20,10,-10", Scale="1,1,1") {
    rotationMode=BMM_Additive
    translationMode=BMM_Additive
    alphaBoolBlend {blendOption=Linear}
    alphaScaleBiasClamp {clampMin=1}
    boneToModify {boneName=dyn_pouch_02}
    rotation=<rot -20 10 -10>
    translation {x=-2, z=3}
   }
   n25 = "Transform (Modify) Bone - Bone: dyn_collar_r_01"(ComponentPose=@n24, Alpha=1, Translation="1.062,2.897,1.584", Rotation="-31.962,-109.569,42.591", Scale="1,1,1") {
    rotationMode=BMM_Additive
    translationMode=BMM_Additive
    alphaBoolBlend {blendOption=Linear}
    boneToModify {boneName=dyn_collar_r_01}
    rotation=<rot -31.962 -109.569 42.591>
    translation=<1.062 2.897 1.584>
   }
   n26 = "Transform (Modify) Bone - Bone: dyn_collar_r_02"(ComponentPose=@n25, Alpha=1, Translation="1.808,2.934,0.402", Rotation="21.043,60.121,46.798", Scale="1,1,1") {
    rotationMode=BMM_Additive
    translationMode=BMM_Additive
    alphaBoolBlend {blendOption=Linear}
    boneToModify {boneName=dyn_collar_r_02}
    rotation=<rot 21.043 60.121 46.798>
    translation=<1.808 2.934 0.402>
   }
   n27 = "Transform (Modify) Bone - Bone: dyn_collar_l_01"(ComponentPose=@n26, Alpha=1, Translation="0,0,0", Rotation="25.253,-13.602,-39.995", Scale="1,1,1") {
    rotationMode=BMM_Additive
    alphaBoolBlend {blendOption=Linear}
    boneToModify {boneName=dyn_collar_l_01}
    rotation=<rot 25.253 -13.602 -39.995>
   }
   n28 = "Transform (Modify) Bone - Bone: dyn_collar_l_02"(ComponentPose=@n27, Alpha=1, Translation="-0.356,0.000,-1.000", Rotation="0.000,0.000,-20.000", Scale="1,1,1") {
    rotationMode=BMM_Additive
    translationMode=BMM_Additive
    alphaBoolBlend {blendOption=Linear}
    boneToModify {boneName=dyn_collar_l_02}
    rotation {roll=-20}
    translation {x=-0.356, z=-1}
   }
   n29 = "Transform (Modify) Bone - Bone: dyn_pouch_02"(ComponentPose=@n28, Alpha=1, Translation="1.073,-1.081,0.000", Rotation="10.000,0.000,0.000", Scale="1,1,1") {
    rotationMode=BMM_Additive
    translationMode=BMM_Additive
    alphaBoolBlend {blendOption=Linear}
    boneToModify {boneName=dyn_pouch_02}
    rotation {pitch=10}
    translation=<1.073 -1.081>
   }
   n30 = Get bIsInFrontEnd()
   n31 = "Transform (Modify) Bone - Bone: neck_01"!(ComponentPose=@n29, bAlphaBoolEnabled:bool=@n30, Translation="0,0,0", Rotation="0,0,0", Scale="1,1.15,1.15") -> (Pose) {
    alphaInputType=Bool
    rotationMode=BMM_Additive
    rotationSpace=BCS_BoneSpace
    scaleMode=BMM_Additive
    scaleSpace=BCS_BoneSpace
    translationMode=BMM_Additive
    translationSpace=BCS_BoneSpace
    alphaBoolBlend {blendOption=Linear}
    boneToModify {boneName=neck_01}
    scale {y=1.15, z=1.15}
   }
   n32 = "Transform (Modify) Bone - Bone: dyn_zipperMSTR"!(ComponentPose=@n31, bAlphaBoolEnabled:bool=@n30, Translation="-1.5,0,0", Rotation="0,0,0", Scale="1,1,1") -> (Pose) {
    alphaInputType=Bool
    rotationMode=BMM_Additive
    rotationSpace=BCS_BoneSpace
    scaleMode=BMM_Additive
    scaleSpace=BCS_BoneSpace
    translationMode=BMM_Additive
    translationSpace=BCS_BoneSpace
    alphaBoolBlend {blendOption=Linear}
    boneToModify {boneName=dyn_zipperMSTR}
    translation {x=-1.5}
   }
   Output Pose(Result=Component To Local(ComponentPose=@n32)) {layerGroup=None}
  ]}
 EventGraph {nodes [
   Event BlueprintUpdateAnimation()
   n1 = BreakTransform(InTransform=GetDeltaTransformFromRefPose(self=GetOwnerSkeletalMeshComponent(), BoneName=head, BaseName=spine_05))
   > Set Head Up(Head Up:double=MapRangeClamped(Value=BreakRotator(InRot=@n1.Rotation).Yaw, InRangeA=-25, InRangeB=-55)) -> (Output_Get:double)
   > n2 = Branch(Condition=Get bIsSkyDiving())
   n3 = GetOwnerSkeletalMeshComponent()
   n4 = BreakRotator(InRot=BreakTransform(InTransform=GetDeltaTransformFromRefPose(self=@n3, BoneName=upperarm_r, BaseName=spine_05)).Rotation)
   Set Arm Up R(execute=@n2.else, Arm Up R:double=MapRangeClamped(Value="Absolute (Float)"(A=@n4.Yaw), InRangeA=0, InRangeB=100)) -> (Output_Get:double)
   n5 = GetOwnerSkeletalMeshComponent()
   n6 = BreakRotator(InRot=BreakTransform(InTransform=GetDeltaTransformFromRefPose(self=@n5, BoneName=upperarm_l, BaseName=spine_05)).Rotation)
   > Set Arm Up L(Arm Up L:double=MapRangeClamped(Value="Absolute (Float)"(A=@n6.Yaw), InRangeA=40, InRangeB=100)) -> (Output_Get:double)
   n7 = GetDeltaTransformFromRefPose(self=GetOwnerSkeletalMeshComponent(), BoneName=thigh_r, BaseName=pelvis)
   n8 = "Absolute (Float)"(A=BreakRotator(InRot=BreakTransform(InTransform=@n7).Rotation).Yaw)
   n9 = "To Float (Boolean)"(InBool=Get bIsCrouching())
   > Set R_ThighRot_Z(R_ThighRot_Z:double=float * float(A=MapRangeClamped(Value=@n8, InRangeA=0, InRangeB=100), B=MapRangeClamped(Value=@n9, InRangeA=0, InRangeB=1, OutRangeA=0.5)) {is=title K2Node_PromotableOperator}) -> (Output_Get:double)
   n10 = BreakTransform(InTransform=GetDeltaTransformFromRefPose(self=GetOwnerSkeletalMeshComponent(), BoneName=thigh_r, BaseName=pelvis))
   MapRangeClamped(Value="Absolute (Float)"(A=BreakRotator(InRot=@n10.Rotation).Pitch), InRangeA=0, InRangeB=180)
   Set R_ThighRot_Y(R_ThighRot_Y:double=0) -> (Output_Get:double)
  ]}
]
variables [
 Arm Up L {type=number}
 Arm Up R {type=number}
 bParentToThigh {type=boolean}
 fSideMovement {type=number}
 fSideMovementAbsolute {type=number}
 Head Up {type=number}
 R_ThighRot_Y {type=number}
 R_ThighRot_Z {type=number}
]
