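# Physics asset for the face accessory mesh of the Blizzard Bomber female medium character cosmetic, defining physics for its fringe bones.
# A single head capsule anchors a chain of small capsules running from the fringe master bone through three fringe bones.
# The master bone is rotationally locked to the head, while each following joint has tight swing and twist limits, with the first joint swinging more freely than the outer two.
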PhysicsAsset /BRCosmetics/Characters/Player/Female/Medium/Bodies/F_MED_Blizzard_Bomber/Meshes/Parts/F_MED_Blizzard_Bomber_FaceAcc_Physics
bodies {
 dyn_fringe_01 [{shapeType=Capsule, length=0.5, radius=1.5, shapeName=dyn_fringe_01_capsule, center=<0 0.2178 1.9229>, rotation=<rot -0.0001 0.0008 -60.0004>}]
 dyn_fringe_02 [{shapeType=Capsule, length=0.5, radius=1.5, shapeName=dyn_fringe_02_capsule, center=<0 -0.143 1.1525>, rotation=<rot -0.0001 0.0009 -70.0002>}]
 dyn_fringe_03 [{shapeType=Capsule, length=0.5, radius=1.5, shapeName=dyn_fringe_03_capsule, center=<0 -0.097 1.279>, rotation=<rot -0.0001 0.0008 -69.9999>}]
 dyn_fringe_MSTR [{
   shapeType=Capsule
   length=0.5
   radius=1.5
   shapeName=dyn_fringe_MSTR_capsule
   center=<0 0.2178 1.9229>
   rotation=<rot -0.0001 0.0006 -60.0003>
  }]
 head [{
   shapeType=Capsule
   length=7.25
   radius=6.8901
   shapeName=head_capsule
   center=<2.7173 2.0759 0.9298>
   rotation=<rot 0.1473 -129.9032 90.0968>
  }]
}
constraints [
 {
  bone1Name=dyn_fringe_MSTR
  bone2Name=head
  swing1LimitDegrees=45
  swing1Motion=Locked
  swing2LimitDegrees=45
  swing2Motion=Locked
  twistLimitDegrees=45
  twistMotion=Locked
 }
 {bone1Name=dyn_fringe_01, bone2Name=dyn_fringe_MSTR, swing1LimitDegrees=20, swing2LimitDegrees=20, twistLimitDegrees=5}
 {bone1Name=dyn_fringe_02, bone2Name=dyn_fringe_01, swing1LimitDegrees=10, swing2LimitDegrees=10, twistLimitDegrees=5}
 {bone1Name=dyn_fringe_03, bone2Name=dyn_fringe_02, swing1LimitDegrees=10, swing2LimitDegrees=10, twistLimitDegrees=5}
]
