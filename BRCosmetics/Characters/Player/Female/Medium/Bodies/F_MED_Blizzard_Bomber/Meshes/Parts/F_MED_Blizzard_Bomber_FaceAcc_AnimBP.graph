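# Post-copy animation Blueprint for the Blizzard Bomber female character's face accessory mesh, adding secondary physics motion.
# It copies the attached parent mesh's pose, then runs RigidBody simulation with the accessory physics asset, overridden gravity and damped component motion.
# A constraint then ties the dyn_fringe_01 bone's position to the head before output.
# Additive hair-bone offset nodes sit unconnected to the output, and the event graph has no logic.
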AnimBlueprint /BRCosmetics/Characters/Player/Female/Medium/Bodies/F_MED_Blizzard_Bomber/Meshes/Parts/F_MED_Blizzard_Bomber_FaceAcc_AnimBP
target_skeleton=&/F_MED_Blizzard_Bomber_FaceAcc_Skeleton
graphs [
 AnimGraph {nodes [
   n1 = "Transform (Modify) Bone - Bone: dyn_hair_02"(Alpha=1, Translation="0,0,0", Rotation="9.847,-1.753,-0.150", Scale="1,1,1") {
    rotationMode=BMM_Additive
    translationMode=BMM_Additive
    alphaBoolBlend {blendOption=Linear}
    boneToModify {boneName=dyn_hair_02}
    rotation=<rot 9.847 -1.753 -0.15>
   }
   "Transform (Modify) Bone - Bone: dyn_hair_03"(ComponentPose=@n1, Alpha=1, Translation="0,-0.000,0.826", Rotation="-31.187,-5.826,-16.597", Scale="1,1,1") {
    rotationMode=BMM_Additive
    translationMode=BMM_Additive
    alphaBoolBlend {blendOption=Linear}
    boneToModify {boneName=dyn_hair_03}
    rotation=<rot -31.187 -5.826 -16.597>
    translation {z=0.826}
   }
   n2 = RigidBody(ComponentPose=Local To Component(LocalPose=Copy Pose From Mesh() {bUseAttachedParent=true})) {
    bClampLinearTranslationLimitToRefPose=true
    bOverrideWorldGravity=true
    overridePhysicsAsset=&/F_MED_Blizzard_Bomber_FaceAcc_Physics
    alphaBoolBlend {blendOption=Linear}
    componentLinearAccScale=<0.1 0.1 0.1>
    overrideWorldGravity {z=-980}
    simSpaceSettings {maxAngularAcceleration=20, maxAngularVelocity=10, maxLinearAcceleration=20, maxLinearVelocity=10, externalLinearDragV=<5 5 5>}
   }
   n3 = "Constraint - dyn_fringe_01 (P)"!(ComponentPose=@n2, Alpha=1, ConstraintWeights_0=1, ConstraintWeights_1:float=0) -> (Pose) {
    alphaBoolBlend {blendOption=Linear}
    boneToModify {boneName=dyn_fringe_01}
    constraintSetup [{targetBone {boneName=head}}]
    constraintWeights [1,0]
   }
   Output Pose(Result=Component To Local(ComponentPose=@n3)) {layerGroup=None}
  ]}
 EventGraph {nodes [Event BlueprintUpdateAnimation()]}
]
variables []
